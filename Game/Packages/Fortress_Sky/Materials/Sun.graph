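# Unlit material for the Fortress Sky sun, producing a bright two-tone glow with separate rim and center colors.
# A Fresnel term weights the rim color and an inverted Fresnel with a different exponent weights the center color; the two are added.
# A static switch selects either fixed colors or morning/day/evening/night colors resolved through Color_Time_of_Day.
# A SunBrightness parameter scales the result.
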Material /Game/Packages/Fortress_Sky/Materials/Sun
graph {
 properties {pixelDepthOffsetMode=PDOM_Legacy, shadingModel=MSM_Unlit, usedShadingModels=Unlit}
 nodes [
  n1 = Multiply(
   A=Fresnel() {baseReflectFraction=0, exponent=2}
   B=MaterialFunctionCall(
    Color_Morning=VectorParameter() {parameterName=Morning Color_1, defaultValue=<rgba 0.88 0.056 0.0132 1>}.RGB
    Color_Day=VectorParameter() {parameterName=Day Color_1, defaultValue=<rgba 1 0.8317 0.8929 1>}.RGB
    Color_Evening=VectorParameter() {parameterName=Evening Color_1, defaultValue=<rgba 1 0.0695 0 1>}.RGB
    Color_Night=VectorParameter() {parameterName=Night Color_1, defaultValue=<rgba 1 0.0695 0 1>}.RGB
   ) {materialFunction=&Color_Time_of_Day}.Result
  )
  n2 = Multiply(
   A=MaterialFunctionCall(
    Color_Morning=VectorParameter() {parameterName=Morning Color_2, defaultValue=<rgba 0.88 0.056 0.0132 1>}.RGB
    Color_Day=VectorParameter() {parameterName=Day Color_2, defaultValue=<rgba 1 0.8317 0.8929 1>}.RGB
    Color_Evening=VectorParameter() {parameterName=Evening Color_2, defaultValue=<rgba 1 0.0695 0 1>}.RGB
    Color_Night=VectorParameter() {parameterName=Night Color_2, defaultValue=<rgba 1 0.0695 0 1>}.RGB
   ) {materialFunction=&Color_Time_of_Day}.Result
   B=OneMinus(None=Fresnel() {baseReflectFraction=0})
  )
  n3 = Multiply(A=VectorParameter() {parameterName=SunColor1, defaultValue=<rgba 0.5098 0.51 0.4769 1>}.RGB, B=Fresnel() {baseReflectFraction=0, exponent=2})
  n4 = Multiply(A=VectorParameter() {parameterName=SunColor2, defaultValue=<rgba 1 0.9896 0.9896 1>}.RGB, B=OneMinus(None=Fresnel() {baseReflectFraction=0}))
  Multiply(
   A=StaticSwitchParameter(True=Add(A=@n1, B=@n2), False=Add(A=@n3, B=@n4)) {parameterName=Change Color over the time of day}
   B=ScalarParameter() {defaultValue=10, parameterName=SunBrightness}
  ) {constA=10}
 ]
}
