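# Material function, exposed to the library under Misc and Fortnite, that produces a screen-space gradient around the sun.
# It reads the sun and moon model world position from the FortniteMaterialParameters collection and normalizes the direction from the camera to it.
# A dot product with the per-pixel camera vector forms the falloff, output both raw (Dot) and negated (Dot*-1), the negated one peaking toward that position.
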MaterialFunction /Game/Packages/Fortress_Sky/Fortnite_SunScreenSpaceFalloff
graph {
 properties {bExposeToLibrary=true, description=Produces a screenspace gradient around the sun, libraryCategoriesText [Misc,Fortnite]}
 nodes [
  n1 = CollectionParameter() {
   collection=&FortniteMaterialParameters
   parameterName=SunAndMoonModelWorldPosition
  }
  n2 = ComponentMask(None=@n1) {b=true, g=true, r=true}
  n3 = CameraPositionWS()
  n4 = DotProduct(A=Normalize(VectorInput=Subtract(A=@n2, B=@n3)), B=CameraVectorWS())
  FunctionOutput(None=Multiply(A=@n4) {constB=-1}) {outputName=Dot*-1}
  FunctionOutput(None=@n4) {outputName=Dot}
 ]
}
