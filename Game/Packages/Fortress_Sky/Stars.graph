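# Unlit sky material for Fortress_Sky, with several star-sky chains; the graph shows no link to a material output.
# The fullest chain adds twinkle-masked stars and a galaxy nebula to a zenith color, then blends in view-based horizon color, sun disc and panning clouds.
# Other unconsumed chains layer noise-jittered star channels with galaxy textures or sun haze, plus a cheaper low-quality star switch.
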Material /Game/Packages/Fortress_Sky/Stars
graph {
 properties {bTangentSpaceNormal=false, pixelDepthOffsetMode=PDOM_Legacy, shadingModel=MSM_Unlit, translucencyPass=MTP_BeforeDOF, usedShadingModels=Unlit}
 nodes [
  n1 = VectorParameter() {parameterName=Sky - Zenith Color, defaultValue=<rgba 0 0 0 1>}
  n2 = TextureSample(UVs=TextureCoordinate() {uTiling=0.4, vTiling=0.4}) {samplerType=SAMPLERTYPE_LinearGrayscale, texture=/Game/Global/Textures/Default/Tileable_Patterns/T_StormCloudDetail01_M}
  n3 = TextureCoordinate() {uTiling=6, vTiling=5}
  n4 = Constant() {r=-1}
  n5 = TextureSample!(UVs=@n3, Tex, Bias=@n4, Apply View MipBias) {mipValueMode=TMVM_MipBias, samplerType=SAMPLERTYPE_Masks, texture=/Game/Packages/Fortress_Sky/Textures/T_Sky_Stars_Sharp}
  n6 = Add(
   A=TextureSample!(UVs=TextureCoordinate(), Tex, Apply View MipBias) {samplerType=SAMPLERTYPE_LinearGrayscale, texture=&T_StormCloudDetail01_M}.R
   B=@n5.R
  )
  n7 = Panner(
   Coordinate=TextureCoordinate() {uTiling=25, vTiling=25}
   Time=MaterialFunctionCall(Speed=Constant() {r=0.245}) {materialFunction=&TimeWithSpeedVariable}.Time
  ) {speedX=0.1, speedY=0.0012}
  n8 = TextureSample(UVs=Add(A=OneMinus(None=@n7)) {constB=0.5}) {samplerType=SAMPLERTYPE_LinearColor, texture=/Game/Global/Textures/Default/Noise/Noise32x32}
  n9 = Multiply(A=Subtract(A=ComponentMask(None=@n8.RGB) {g=true, r=true}) {constA=0, constB=0.5}) {constB=0.005}
  n10 = Add(A=Multiply(A=@n9) {constB=0.5}, B=@n3)
  n11 = Add(
   A=Multiply(A=@n5.RGB, B=Constant3Vector() {constant=<rgba 1 0 0 0>})
   B=Multiply(
    A=TextureSample!(UVs=@n10, Tex, Bias=@n4, Apply View MipBias) {mipValueMode=TMVM_MipBias, samplerType=SAMPLERTYPE_Masks, texture=&T_Sky_Stars_Sharp}.RGB
    B=Constant3Vector() {constant=<rgba 0 1 0 0>}
   )
  )
  n12 = Add(A=@n3, B=@n9)
  n13 = Multiply(
   A=TextureSample(UVs=@n12, Bias=@n4) {mipValueMode=TMVM_MipBias, samplerType=SAMPLERTYPE_Masks, texture=&T_Sky_Stars_Sharp}.RGB
   B=Constant3Vector() {constant=<rgba 0 0 1 0>}
  )
  n14 = Multiply(A=Multiply(A=Add(A=@n11, B=@n13), B=Multiply(A=@n8.B, B=@n8.B))) {constB=1.5}
  n15 = Multiply(A=@n14, B=@n14)
  n16 = Add(A=Multiply(A=Multiply(A=@n2.R, B=@n6), B=@n15), B=@n15)
  n17 = Add(A=@n16, B=@n16)
  n18 = Add(A=@n12, B=@n12)
  n19 = Add(A=OneMinus(None=@n18)) {constB=0.25}
  n20 = Add(
   A=TextureSample(UVs=@n18, Bias=@n4) {mipValueMode=TMVM_MipBias, samplerType=SAMPLERTYPE_Masks, texture=&T_Sky_Stars_Sharp}.RGB
   B=Multiply(A=TextureSample(UVs=@n19, Bias=@n4) {mipValueMode=TMVM_MipBias, samplerType=SAMPLERTYPE_Masks, texture=&T_Sky_Stars_Sharp}.RGB) {constB=0.5}
  ) {constB=0}
  n21 = ScalarParameter() {defaultValue=0.1, parameterName=Sky - Star Brightness}
  n22 = Add(A=Multiply(A=@n16) {constB=0.025}, B=Multiply(A=Add(A=@n17, B=@n20), B=@n21))
  n23 = ComponentMask(None=TextureCoordinate() {coordinateIndex=1}) {g=true}
  n24 = Multiply(
   A=Add(A=OneMinus(None=@n23), B=ScalarParameter() {defaultValue=-0.1, parameterName=Sky - Zenith to Horizon Bias}) {constB=0}
   B=ScalarParameter() {defaultValue=0.5, parameterName=Sky - Zenith to Horizon Scale}
  )
  n25 = ScalarParameter() {defaultValue=0.7454}
  n26 = ScalarParameter() {defaultValue=3.1245}
  n27 = Add(
   A=Multiply(
    A=TextureCoordinate() {uTiling=4, vTiling=2}
    B=MaterialFunctionCall(X=@n25, Y=@n26) {materialFunction=&MakeFloat2}.Result
   )
   B=MaterialFunctionCall(X=ScalarParameter(), Y=ScalarParameter() {defaultValue=-2.1853}) {materialFunction=&MakeFloat2}.Result
  )
  n28 = TextureSample!(UVs=@n27, Tex, Apply View MipBias) {texture=/Game/Packages/Fortress_Sky/Textures/Galaxy_Background}
  n29 = Constant() {r=0.5}
  n30 = SphereMask(
   A=MaterialFunctionCall(Float2=@n27) {materialFunction=&BreakOutFloat2Components}.G
   B=@n29
   Radius=@n29
  )
  n31 = Multiply(A=Multiply(A=TextureCoordinate() {uTiling=4, vTiling=2}, B=ScalarParameter() {defaultValue=1.0645}), B=Constant2Vector() {g=3, r=1})
  Add(
   A=LinearInterpolate(
    A=VectorParameter() {parameterName=Sky - Horizon color, defaultValue=<rgba 0.6638 0.2771 0.5312 0>}.RGB
    B=Add(A=@n22, B=@n1.RGB)
    Alpha=Clamp(None=@n24)
   )
   B=Add(A=Multiply(A=@n28.RGB, B=@n30), B=TextureSample(UVs=@n31) {texture=/Game/Packages/Fortress_Sky/Textures/Galaxy_Background_Tiling}.RGB)
  )
  Add()
  n32 = MaterialFunctionCall() {materialFunction=&/Fortnite_SunScreenSpaceFalloff}
  n33 = ScalarParameter() {defaultValue=0.75, parameterName=Sun Haze Bias}
  n34 = Clamp(None=Multiply(A=Add(A=@n32.Dot*-1, B=@n33), B=ScalarParameter() {defaultValue=0.5, parameterName=Sun Haze Scale}))
  n35 = ScalarParameter() {defaultValue=2, parameterName=Sun Haze Power}
  Add(
   A=@n22
   B=Multiply(A=VectorParameter() {parameterName=Sun Haze Color, defaultValue=<rgba 0.2186 0.325 0.681 1>}.RGB, B=Power(Base=@n34, Exp=@n35))
  )
  n36 = VectorParameter() {group=Sky, parameterName=Zenith Color, defaultValue=<rgba 0.0243 0.0439 0.1 1>}
  n37 = TextureCoordinate() {coordinateIndex=1, uTiling=12, vTiling=6}
  n38 = Multiply(A=TextureCoordinate() {coordinateIndex=1, uTiling=6, vTiling=3}, B=ScalarParameter() {defaultValue=1.481, parameterName=TilingTwinkleMask})
  n39 = Multiply(A=ScalarParameter() {defaultValue=6.9292, parameterName=Speed}, B=Time())
  n40 = Multiply(
   A=TextureSample(UVs=@n37) {samplerType=SAMPLERTYPE_Masks, texture=&T_Sky_Stars_Sharp}.RGB
   B=TextureSample(UVs=Panner(Coordinate=@n38, Time=@n39) {speedX=0.01, speedY=0.008}) {samplerType=SAMPLERTYPE_Masks, texture=/Game/Packages/Fortress_Sky/Textures/T_Sky_Stars_TwinkleMask}.R
  )
  n41 = ScalarParameter() {defaultValue=1, parameterName=Stars brightness}
  n42 = ScalarParameter() {defaultValue=0.4107, parameterName=NebularScaleX}
  n43 = ScalarParameter() {defaultValue=0.9266, parameterName=NebularScaleY}
  n44 = Add(
   A=Multiply(A=@n37, B=AppendVector(A=@n42, B=@n43))
   B=AppendVector(
    A=ScalarParameter() {defaultValue=0.9266, parameterName=NebularPanX_1}
    B=ScalarParameter() {defaultValue=0.6938, parameterName=NebularPanY_1}
   )
  )
  n45 = Add(
   A=Multiply(A=Multiply(A=@n40, B=@n41) {constB=0.02}, B=ScalarParameter() {defaultValue=1, parameterName=Sun height})
   B=Multiply(
    A=TextureSample(UVs=@n44) {texture=&Galaxy_Background}.RGB
    B=ScalarParameter() {defaultValue=1.3398, parameterName=NebularBrightness}
   )
  )
  n46 = Clamp(None=DotProduct(A=CameraVectorWS(), B=Constant3Vector() {constant=<rgba 0 0 -1 0>}))
  n47 = Power(Base=OneMinus(None=@n46), Exp=ScalarParameter() {defaultValue=3, group=Sky, parameterName=Horizon Falloff})
  n48 = DotProduct(A=CameraVectorWS(), B=Normalize(VectorInput=VectorParameter() {group=Sun, parameterName=Light direction, defaultValue=<rgba -2 0 -1 1>}.RGB))
  n49 = Constant() {r=1}
  n50 = ScalarParameter() {defaultValue=0.0003, group=Sun, parameterName=Sun Radius}
  n51 = VectorParameter() {group=Sun, parameterName=Sun color, defaultValue=<rgba 1 0.8 0.4 1>}
  n52 = ScalarParameter() {defaultValue=50, parameterName=Sun brightness}
  n53 = Add(
   A=LinearInterpolate(
    A=Add(A=@n36.RGB, B=@n45)
    B=VectorParameter() {group=Sky, parameterName=Horizon color, defaultValue=<rgba 0.4703 0.5 0.5 1>}.RGB
    Alpha=Clamp(None=@n47)
   )
   B=Multiply(A=SphereMask(A=@n48, B=@n49, Radius=@n50) {attenuationRadius=0.001, hardnessPercent=0}, B=Multiply(A=@n51.RGB, B=@n52))
  )
  n54 = Time()
  n55 = ScalarParameter() {defaultValue=0.1, group=Animation, parameterName=Cloud speed}
  n56 = TextureSample(UVs=Panner(Time=Multiply(A=@n54, B=@n55)) {speedX=0.0002}) {texture=/Engine/EngineSky/T_Sky_Blue}
  n57 = TextureSample(UVs=Panner(Time=Multiply(A=@n54, B=@n55)) {speedX=0.001}) {samplerType=SAMPLERTYPE_Grayscale, texture=/Engine/EngineSky/T_Sky_Clouds_M}
  n58 = Divide(A=Subtract(A=WorldPosition().XYZ, B=ObjectPositionWS()), B=Multiply(A=ObjectRadius()) {constB=-0.1}) {constB=2000}
  n59 = OneMinus(None=Clamp(None=ComponentMask(None=@n58) {b=true}))
  n60 = ScalarParameter() {defaultValue=1, parameterName=Cloud opacity}
  n61 = TextureCoordinate() {uTiling=0.5, vTiling=0.5}
  n62 = Power(
   Base=LinearInterpolate(B=LinearInterpolate(A=@n56.R, B=@n57.R, Alpha=@n46), Alpha=Multiply(A=@n59, B=@n60))
   Exp=LinearInterpolate(
    A=ScalarParameter() {defaultValue=1, group=Animation, parameterName=NoisePower1}
    B=ScalarParameter() {defaultValue=4, group=Animation, parameterName=NoisePower2}
    Alpha=TextureSample(UVs=@n61) {samplerType=SAMPLERTYPE_Grayscale, texture=&T_Sky_Clouds_M}.R
   ) {constA=0.1, constB=4}
  )
  n63 = Constant() {r=1}
  n64 = Constant() {r=1.3}
  n65 = Clamp(None=Power(Base=SphereMask(A=@n48, B=@n63, Radius=@n64) {attenuationRadius=0, hardnessPercent=0}) {constExponent=10})
  n66 = Multiply(A=@n62, B=@n62) {constB=0.5}
  n67 = Add(
   A=Multiply(A=VectorParameter() {group=CloudColors, parameterName=Cloud color, defaultValue=<rgba 0.7168 0.7822 0.885 0>}.RGB, B=@n62)
   B=Multiply(A=Multiply(A=@n65, B=@n51.RGB), B=Multiply(A=@n66) {constB=0.4})
  )
  n68 = Clamp(None=@n66)
  Multiply(A=Multiply(
    A=LinearInterpolate(A=@n53, B=@n67, Alpha=@n68)
    B=VectorParameter() {group=Global, parameterName=Overall Color, defaultValue=<rgba 1 1 1 1>}.RGB
   )) {constB=1.5}
  QualitySwitch(Default=@n22, Low=Multiply(A=Multiply(A=@n21, B=@n5.RGB)) {constB=2}, High=@n22)
  SphereMask()
  TextureSample!(UVs=TextureCoordinate() {uTiling=12, vTiling=6}, Tex, Bias, Apply View MipBias) {mipValueMode=TMVM_MipBias, samplerType=SAMPLERTYPE_Masks, texture=&T_Sky_Stars_Sharp}
 ]
}
