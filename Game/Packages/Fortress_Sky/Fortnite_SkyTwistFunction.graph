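# Material function that produces animated, swirl-distorted UVs, apparently for Fortnite skies, by twisting a flow-panned UV grid cell by cell.
# Alternate columns shift half a cell, and a squared sphere mask drives rotation around each cell center, strongest in the middle.
# A repeating pseudo-random cell value flips twist direction and phases a sine that scales the rotated UVs.
# Low quality drops both variations; extra outputs expose angle, grid UVs, random values and edge falloff.
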MaterialFunction /Game/Packages/Fortress_Sky/Fortnite_SkyTwistFunction
graph {
 properties {bExposeToLibrary=true, libraryCategoriesText [Misc,Fortnite]}
 nodes [
  n1 = FunctionInput() {bUsePreviewValueAsDefault=true, inputName=Time offset, inputType=FunctionInput_Scalar, previewValue=<0 1 0 1>}
  n2 = FunctionInput(Preview=Time()) {bUsePreviewValueAsDefault=true, inputName=un frac Time, inputType=FunctionInput_Scalar, previewValue=<0 1 0 1>}
  n3 = FunctionInput() {bUsePreviewValueAsDefault=true, inputName=Flow Direction, inputType=FunctionInput_Vector2, previewValue=<0 1 0 1>}
  n4 = FunctionInput() {bUsePreviewValueAsDefault=true, inputName=UVS * size, inputType=FunctionInput_Vector2, previewValue=<128 128 0 1>}
  n5 = Add(A=Multiply(A=Add(A=@n1, B=@n2), B=@n3), B=@n4) {constA=0.5}
  n6 = ComponentMask(None=@n5) {r=true}
  n7 = Ceil(None=Add(A=@n6) {constB=0.5})
  n8 = ComponentMask(None=@n5) {g=true}
  n9 = AppendVector(A=@n6, B=Add(A=Multiply(A=@n7) {constB=0.5}, B=Multiply(A=@n8)))
  n10 = Add(A=@n9) {constB=0.5}
  n11 = SphereMask(A=Frac(None=@n10), B=Constant() {r=0.5}) {attenuationRadius=0.5, desc=23, hardnessPercent=0}
  n12 = Multiply(A=@n11, B=@n11)
  n13 = FunctionInput() {bUsePreviewValueAsDefault=true, inputName=Grid Noise Tile Rep, inputType=FunctionInput_Vector2, previewValue=<5 3 0 1>}
  n14 = Frac(None=Divide(A=Add(A=@n9) {constB=-0.5}, B=@n13) {constB=5}) {desc=mod}
  n15 = Divide(A=Floor(None=Multiply(A=@n14, B=@n13) {constB=10}), B=@n13) {desc=0-1 values that have the number of increments specified}
  n16 = Frac(None=Add(A=ComponentMask(None=@n15) {g=true}, B=ComponentMask(None=@n15) {r=true}))
  n17 = LinearInterpolate(Alpha=Ceil(None=Subtract(A=@n16) {constB=0.5})) {constA=-1}
  n18 = Multiply(A=@n12, B=@n17) {desc=4 ins to randomize}
  n19 = FunctionInput() {bUsePreviewValueAsDefault=true, inputName=Rotation amount modulator, inputType=FunctionInput_Scalar, previewValue=<0.5 0 0 1>}
  n20 = Multiply(A=QualitySwitch(Default=@n18, Low=@n12, High=@n18), B=@n19) {constA=1}
  FunctionOutput(None=@n20) {outputName=RotationAngleValues}
  n21 = Ceil(None=@n10)
  n22 = Constant()
  n23 = RotateAboutAxis(
   NormalizedRotationAxis=Constant3Vector() {constant=<rgba 0 0 1 0>}
   RotationAngle=@n20
   PivotPoint=AppendVector(A=Subtract(A=@n21) {constB=0.5}, B=@n22)
   Position=AppendVector(A=@n10, B=@n22)
  )
  n24 = FunctionInput() {bUsePreviewValueAsDefault=true, inputName=SineAnimSpeed, inputType=FunctionInput_Scalar, previewValue=<0.3 3 0 1>}
  n25 = Multiply(A=@n2, B=@n24)
  n26 = FunctionInput() {bUsePreviewValueAsDefault=true, inputName=SineAnimOffset, inputType=FunctionInput_Scalar, previewValue=<0 3 0 1>}
  n27 = Sine(None=Add(A=@n16, B=Add(A=@n25, B=@n26)) {desc=offset wave phase})
  n28 = Multiply(A=@n23, B=ConstantBiasScale(None=@n27) {bias=2, scale=0.25}) {desc=+15 for random height}
  n29 = Multiply(A=@n23) {constB=0.75}
  FunctionOutput(None=ComponentMask(None=QualitySwitch(Default=@n28, Low=@n29, High=@n28)) {g=true, r=true}) {outputName=Rotated UVs, sortPriority=1}
  FunctionOutput(None=@n10) {outputName=NIU - Modified UVOutput, sortPriority=10}
  FunctionOutput(None=ConstantBiasScale(None=@n27)) {outputName=NIU - Random values per block, sortPriority=10}
  n30 = FunctionInput() {bUsePreviewValueAsDefault=true, inputName=EdgeMaskSize, inputType=FunctionInput_Scalar, previewValue=<20 0 0 1>}
  n31 = Frac(None=Add(A=Constant() {r=0.5}, B=Frac(None=@n9)))
  n32 = AppendVector(A=Multiply(A=@n30, B=OneMinus(None=@n31)) {constA=10}, B=Multiply(A=@n30, B=@n31) {constA=10})
  FunctionOutput(None=Clamp(None=MaterialFunctionCall!(f2, f3, f4=@n32) {materialFunction=/Game/Packages/Fortnite_UniqueMaterials_1/VectorComponentOps}.f4 Min)) {outputName=NIU - WPO_Falloff, sortPriority=10}
  FunctionOutput(None=@n17) {outputName=NIU - Random values negative or positive sign, sortPriority=10}
 ]
}
